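# Niagara system for the main player's Reverence emote, combining a burst of dusty smoke with a few sparks.
# Collision_Smoke bursts drifting sprites that grow, spin and fade, with curves and randomness feeding dynamic material parameters.
# NE_FN_Sparks launches gravity-pulled sprites in an upward cone; a random SparkIntensity sets color and mass, and LOD distance enlarges them.
# User.Overheat boosts spark speed; both front-end bools are exposed but unreferenced here.
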NiagaraSystem /BRCosmetics/Animation/Game/MainPlayer/Emotes/Reverence/NS_Reverence_Emote
dependencies {
 usedDataInterfaces [&NiagaraDataInterfaceCurve, &NiagaraDataInterfaceColorCurve]
 usedDynamicInputs [&VectorFromFloat]
 usedModules [&SystemState]
}
emitters {
 Collision_Smoke {
  emitterUpdateScript [
   &EmitterState()
   &SpawnBurst_Instantaneous(Spawn Count=10, Use Spawn Probability=false, UseLoopCountLimit=false)
   &SpawnRate(SpawnRate=20, Use Spawn Probability=false) {enabled=false}
  ]
  particleSpawnScript [
   &InitializeParticle(
    Color=&RandomRangeLinearColor(Minimum=<rgba 0.4115 0.3286 0.2562 0.5>, Maximum=<rgba 0.2083 0.1441 0.0948 0.5>)
    Lifetime Max=0.7
    Lifetime Min=1.4
    Lifetime Mode=::NewEnumerator1
    Sprite Rotation Mode=::NewEnumerator2
    Sprite Size Mode=::NewEnumerator1
    Uniform Sprite Size Max=3
    Uniform Sprite Size Min=1.5
    UsePositionOffset=false
    Write Material Random=true
   ) {module=InitializeParticle002}
   &ShapeLocation(MaskSpawnGroup=false, Override Seed=false, Sphere Radius=3, UseOwnerScale=true)
   &AddVelocityFromPoint(Module.Use Falloff Distance=false, Velocity Strength=&RandomRangeFloat(Minimum=40, Maximum=100)) {enabled=false, module=AddVelocityFromPoint001}
   &AddVelocity(Velocity=&RandomRangeVector(Minimum=<80 80 10>, Maximum=<-40 -40 20>))
  ]
  particleUpdateScript [
   &ParticleState()
   &GravityForce(Gravity=<0 0 30>)
   &Drag(Drag=2, Use Linear Drag=true, Use Rotational Drag=false)
   &SolveForcesAndVelocity(Clamp Velocity=false, Limit Acceleration=false)
   &ScaleSpriteSize(Scale Factor=&Vector2DFromFloat(Value=&FloatFromCurve(FloatCurve=DataInterface() {propertyValues {Curve {keys [
          {arriveTangent=87.4549, interpMode=RCIM_Cubic, leaveTangent=87.4549, tangentMode=RCTM_User, time=0, value=1}
          {arriveTangent=4.8897, interpMode=RCIM_Cubic, leaveTangent=4.8897, tangentMode=RCTM_User, time=0.1, value=8}
          {interpMode=RCIM_Cubic, leaveTangent=-6.7553, time=1.0011, value=10}
         ]}}})))
   &ScaleColor(
    Scale Alpha=&FloatFromCurve(
     FloatCurve=DataInterface() {propertyValues {Curve {keys [
         {arriveTangent=17.5364, interpMode=RCIM_Cubic, leaveTangent=17.5364, tangentMode=RCTM_User, time=0, value=0}
         {arriveTangent=-3.1757, interpMode=RCIM_Cubic, tangentMode=RCTM_Break, time=0.1, value=1}
         {arriveTangent=0.0071, interpMode=RCIM_Cubic, leaveTangent=0.0071, tangentMode=RCTM_User, time=0.4, value=1}
         {arriveTangent=-1.6667, interpMode=RCIM_Cubic, leaveTangent=-1.6667, time=0.7, value=0.25}
         <at 1 0 RCIM_Cubic>
        ]}}}
     Scale Curve=0.5
    )
    ScaleA=true
    ScaleRGB=true
   ) {module=ScaleColor001}
   &DynamicMaterialParameters(
    Index 0 Param 1=&Add_Float(
     A=&FloatFromCurve(FloatCurve=DataInterface() {propertyValues {Curve {keys [{arriveTangent=2.1549, time=0, value=0.8}, <at 1 1>]}}})
     B=&Multiply_Float(A=$Particles.Initial.MaterialRandom, B=0.5)
    )
    Index 0 Param 3=&FloatFromCurve(
     FloatCurve=DataInterface() {propertyValues {Curve {keys [{arriveTangent=-0.0944, interpMode=RCIM_Cubic, time=0, value=0}, <at 0.5008 0>]}}}
     Scale Curve=3
    )
    Index 0 Param 4=1
    Index 1 Param 1=1
    Index 1 Param 2=1
    Index 1 Param 3=1
    Index 1 Param 4=1
    Index 2 Param 1=&FloatFromCurve(FloatCurve=DataInterface() {propertyValues {Curve {keys [
         {arriveTangent=2.7333, interpMode=RCIM_Cubic, leaveTangent=2.7333, tangentMode=RCTM_User, time=0, value=0}
         {arriveTangent=-0.0004, interpMode=RCIM_Cubic, leaveTangent=-0.0004, tangentMode=RCTM_User, time=1, value=1}
        ]}}})
    Index 2 Param 2=&RandomRangeFloat(Minimum=0.5, Maximum=1)
    Index 2 Param 3=&FloatFromCurve(FloatCurve=DataInterface() {propertyValues {Curve {keys [
         <at 0 10>
         {arriveTangent=5.218, interpMode=RCIM_Cubic, leaveTangent=5.218, tangentMode=RCTM_User, time=0.1, value=0}
         {arriveTangent=20.2637, interpMode=RCIM_Cubic, leaveTangent=20.2637, tangentMode=RCTM_User, time=1, value=10}
        ]}}})
    Index 2 Param 4=&RandomRangeFloat(Minimum=0.8, Maximum=1.2)
    Module.Param0WriteEnabled=true
    Module.Param0WriteEnabled1=true
    Module.Param0WriteEnabled2=true
    Module.Param0WriteEnabled3=true
    Module.Param1WriteEnabled=true
    Module.Param1WriteEnabled1=true
    Module.Param1WriteEnabled2=true
    Module.Param1WriteEnabled3=true
    Module.Param2WriteEnabled=true
    Module.Param2WriteEnabled1=true
    Module.Param2WriteEnabled2=true
    Module.Param2WriteEnabled3=true
    Module.Param3WriteEnabled=true
    Module.Param3WriteEnabled1=true
    Module.Param3WriteEnabled2=true
    Module.Param3WriteEnabled3=true
    Write Parameter Index 1=true
    Write Parameter Index 2=true
    Write Parameter Index 3=true
   ) {module=DynamicMaterialParameters002}
   &CameraOffset(Camera Offset Amount=50)
   &SpriteRotationRate(Rotation Rate=&RandomRangeFloat(Minimum=-90, Maximum=90))
  ]
  rendererClasses [&NiagaraSpriteRendererProperties]
  renderers [{rendererClass=&NiagaraSpriteRendererProperties}]
 }
 NE_FN_Sparks {
  emitterUpdateScript [
   &EmitterState()
   &SpawnBurst_Instantaneous(Spawn Count=&RandomRangeInt(Minimum=4, Maximum=8), Use Spawn Probability=false)
   &SpawnRate(SpawnRate=20, Use Spawn Probability=false) {enabled=false, module=SpawnRate001}
   &SpawnRate(SpawnRate=100, Use Spawn Probability=false) {enabled=false}
  ]
  particleSpawnScript [
   /SetVariables(Particles.SparkIntensity:NiagaraFloat=&FloatFromCurve(
     FloatCurve=DataInterface() {propertyValues {Curve {keys [
         {arriveTangent=0.0774, interpMode=RCIM_Cubic, leaveTangent=0.0774, tangentMode=RCTM_User, time=0.1, value=1}
         {interpMode=RCIM_Cubic, tangentMode=RCTM_User, time=1, value=0}
        ]}}}
     CurveIndex=&RandomRangeFloat(Minimum=0, Maximum=1)
    )) {isSetParameters=true}
   &InitializeParticle(
    Color=&LerpLinearColors(StartColor=<rgba 1 0.9819 0.6157 0.3>, EndColor=<rgba 0.4062 0.3489 0.2962 0.2>, LerpFactor=$Particles.SparkIntensity)
    Lifetime Max=0.5
    Lifetime Min=0.2
    Lifetime Mode=::NewEnumerator1
    Mass=&Lerp_Float(A=0.3, B=1, Alpha=$Particles.SparkIntensity)
    Mass Mode=::NewEnumerator1
    Position Mode=::NewEnumerator1
    Sprite Size=<1 4>
    Sprite Size Mode=::NewEnumerator4
    UsePositionOffset=false
    Write Material Random=true
   )
   /SetVariables(Particles.SpriteSize=&Multiply_Vector2(
     A=$Particles.SpriteSize
     B=&Vector2DFromFloat(Value=&FloatFromCurve(FloatCurve=DataInterface() {propertyValues {Curve {keys [<at 1000 1 RCIM_Cubic>,<at 5000 2>]}}}, CurveIndex=$Engine.Owner.LODDistance))
    )) {isSetParameters=true}
   &SphereLocation(
    MaskSpawnGroup=false
    Override Seed=false
    Sphere Radius=0.7717
    Surface Only=false
    UseNonUniformScale=false
    UseOffset=false
    UseSphereOrientation=false
   ) {enabled=false}
   &SkeletalMeshLocation(MaskSpawnGroup=false, Mesh Sampling Type=::NewEnumerator5, Skeletal Mesh=$User.BackpackSkeleton, Socket Sampling Mode=::NewEnumerator5) {enabled=false}
   &AddVelocityInCone(
    Cone Axis=<0 0 1>
    Module.Use Velocity Falloff On Cone Axis=false
    Velocity Strength=&Multiply_Float(
     A=&RandomRangeFloat(Minimum=50, Maximum=150)
     B=&Lerp_Float(A=1, B=1.5, Alpha=$User.Overheat)
    )
   )
   &DynamicMaterialParameters(
    Index 0 Param 1=0.5
    Index 0 Param 3=1
    Module.Param0WriteEnabled=true
    Module.Param1WriteEnabled=true
    Module.Param2WriteEnabled=true
    Module.Param3WriteEnabled=true
   )
   &ShapeLocation(MaskSpawnGroup=false, Offset Mode=::NewEnumerator0, Override Seed=false, Sphere Radius=1, UseOwnerScale=true)
  ]
  particleUpdateScript [
   &ParticleState()
   &GravityForce(Gravity=&RandomRangeVector(Minimum=<0 0 -200>, Maximum=<0 0 -400>))
   &Drag(Drag=0.15, Use Linear Drag=true, Use Rotational Drag=false)
   &VectorNoiseForce(
    Force Amount=&FloatFromCurve(
     FloatCurve=DataInterface() {propertyValues {Curve {keys [
         {arriveTangent=-9.4515, interpMode=RCIM_Cubic, leaveTangent=-4.1687, tangentMode=RCTM_User, time=0, value=1}
         {arriveTangent=-3.4372, interpMode=RCIM_Cubic, leaveTangent=-3.4372, tangentMode=RCTM_User, time=0.6, value=0}
        ]}}}
     Scale Curve=3000
    )
    Module.Use Random Noise Time Interval=false
   ) {enabled=false}
   &SolveForcesAndVelocity(Clamp Velocity=false, Limit Acceleration=false)
   &ScaleColor(
    Linear Color Curve=DataInterface() {propertyValues {
      AlphaCurve {keys [<at 0 1>,<at 1 1>]}
      BlueCurve {keys [
        <at 0 2 RCIM_Cubic>
        {arriveTangent=-1.452, interpMode=RCIM_Cubic, leaveTangent=-1.452, tangentMode=RCTM_User, time=0.1467, value=1}
        <at 1 0.138>
       ]}
      GreenCurve {keys [
        <at 0 2 RCIM_Cubic>
        {arriveTangent=-1.4408, interpMode=RCIM_Cubic, leaveTangent=-1.4408, tangentMode=RCTM_User, time=0.1467, value=1}
        <at 1 0>
       ]}
      RedCurve {keys [
        <at 0 2 RCIM_Cubic>
        {arriveTangent=-0.8332, interpMode=RCIM_Cubic, leaveTangent=-0.8332, tangentMode=RCTM_User, time=0.1467, value=1}
        <at 1 0.25>
       ]}
     }}
    Scale Mode=::NewEnumerator2
   )
   &ScaleSpriteSize(Scale Factor=&Vector2DFromFloat(Value=&FloatFromCurve(FloatCurve=DataInterface() {propertyValues {Curve {keys [<at 0.4 1 RCIM_Cubic>, {arriveTangent=-3.8662, interpMode=RCIM_Cubic, leaveTangent=-3.8662, tangentMode=RCTM_User, time=1, value=0}]}}})))
   &ScaleSpriteSizeBySpeed(
    Max Scale Factor=<1 1>
    Sample Scale Factor By Curve=true
    Scale Factor Curve=DataInterface() {propertyValues {Curve {keys [{arriveTangent=2.2765, interpMode=RCIM_Cubic, leaveTangent=2.2765, tangentMode=RCTM_User, time=1, value=1}]}}}
   ) {enabled=false}
  ]
  rendererClasses [&NiagaraSpriteRendererProperties]
  renderers [{rendererClass=&NiagaraSpriteRendererProperties}]
 }
}
summary {
 systemName=NS_Reverence_Emote
 emitters [
  {emitterName=Collision_Smoke, rendererClasses [&NiagaraSpriteRendererProperties]}
  {emitterName=NE_FN_Sparks, rendererClasses [&NiagaraSpriteRendererProperties]}
 ]
 userVariables [
  User.bIsFrontEnd {type=NiagaraBool}
  User.bIsFrontEndPreview {type=NiagaraBool}
  User.Overheat {type=NiagaraFloat, defaultValue=0, description=""}
 ]
}
